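# Physics asset for the Winter Hunter backpack cosmetic, giving collision bodies and joints to the pack and its hanging attachments.
# A capsule pack body anchors the skull, zipper and pickaxe through angular-limited constraints, plus a three-link handle.
# A five-link chain hangs from the pickaxe.
# Chain and handle roots are tiny boxes with free swing and twist, while their later links use limited joints, tighter along the chain.
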PhysicsAsset /BRCosmetics/Accessories/FORT_Backpacks/Backpack_M_MED_Winter_Hunter/Meshes/M_MED_Winter_Hunter_Physics
bodies {
 chain1 [{shapeType=Box, extentX=0.5, extentY=0.5, extentZ=0.5, center=<0.5413 -0.1204 -0.2764>}]
 chain2 [{shapeType=Box, extentX=2.2756, extentY=7.8803, extentZ=2.7287, center=<1.2641 0.3893 0.1077>}]
 chain3 [{shapeType=Box, extentX=2.1909, extentY=8.4167, extentZ=2.6442, center=<1.318 0.0025 -0.0028>}]
 chain4 [{shapeType=Box, extentX=2.3094, extentY=7.7721, extentZ=2.6881, center=<1.3104 -0.0363 0.2155>}]
 chain5 [{shapeType=Capsule, length=4.7005, radius=1.9758, center=<1.9795 0.049 0.1649>, rotation=<rot 0 0 -90>}]
 handle1 [{shapeType=Box, extentX=0.5, extentY=0.5, extentZ=0.5, center=<0.7604 0.1046 -0.0037>}]
 handle2 [{shapeType=Box, extentX=1.3894, extentY=1.3813, extentZ=6.068, center=<0.8254 0.2679 0.049>}]
 handle3 [{shapeType=Box, extentX=1.51, extentY=1.51, extentZ=5.51, center=<0.8999 0.1547 0>}]
 pack [{shapeType=Capsule, length=12.75, radius=8.75, center=<-0.1387 -6.5275 -10.4119>}]
 pickaxe [{shapeType=Box, extentX=8.0214, extentY=12.722, extentZ=21.9658, center=<7.6784 0.0039 1.8642>, rotation=<rot -15.383 -85.4839 -49.1055>}]
 skull [{shapeType=Box, extentX=2.2286, extentY=11.1635, extentZ=9.8851}]
 zipper [{shapeType=Box, extentX=3.6228, extentY=1.5438, extentZ=0.5102, center=<1.3203 -0.0308 0.0999>}]
}
constraints [
 {bone1Name=skull, bone2Name=pack, swing1LimitDegrees=45, swing2LimitDegrees=45, twistLimitDegrees=15}
 {bone1Name=zipper, bone2Name=pack, swing1LimitDegrees=15, swing2LimitDegrees=15, twistLimitDegrees=5}
 {bone1Name=pickaxe, bone2Name=pack, swing1LimitDegrees=45, swing2LimitDegrees=45, twistLimitDegrees=15}
 {
  bone1Name=chain1
  bone2Name=pickaxe
  swing1LimitDegrees=45
  swing1Motion=Free
  swing2LimitDegrees=45
  swing2Motion=Free
  twistLimitDegrees=45
  twistMotion=Free
 }
 {bone1Name=chain2, bone2Name=chain1, swing1LimitDegrees=25, swing2LimitDegrees=10, twistLimitDegrees=5}
 {bone1Name=chain3, bone2Name=chain2, swing1LimitDegrees=25, swing2LimitDegrees=10, twistLimitDegrees=5}
 {bone1Name=chain4, bone2Name=chain3, swing1LimitDegrees=25, swing2LimitDegrees=10, twistLimitDegrees=5}
 {bone1Name=chain5, bone2Name=chain4, swing1LimitDegrees=25, swing2LimitDegrees=10, twistLimitDegrees=5}
 {
  bone1Name=handle1
  bone2Name=pack
  swing1LimitDegrees=45
  swing1Motion=Free
  swing2LimitDegrees=45
  swing2Motion=Free
  twistLimitDegrees=45
  twistMotion=Free
 }
 {bone1Name=handle2, bone2Name=handle1, swing1LimitDegrees=45, swing2LimitDegrees=45, twistLimitDegrees=5}
 {bone1Name=handle3, bone2Name=handle2, swing1LimitDegrees=45, swing2LimitDegrees=45, twistLimitDegrees=5}
]
